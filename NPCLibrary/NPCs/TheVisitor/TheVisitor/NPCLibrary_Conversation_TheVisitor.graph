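# Conversation database for The Visitor NPC, the root dialogue reached through its entry tag.
# A speech node picks a greeting by priority from first-meeting, outfit-reaction and per-match conversation-count lines, with a default introduction.
# It then branches through link nodes into separate conversation entries for NPC quests, basic and random services, the Visitor's own dialogue, and Season 39 SynthContact story and weekly quest content.
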ConversationDatabase /NPCLibrary/NPCs/TheVisitor/TheVisitor/NPCLibrary_Conversation_TheVisitor
entry_points [{tag=Conversation.Entry.TheVisitor, destinations [#1]}]
nodes [
 ConversationEntryPointNode_0 {
  id=#1
  class=ConversationEntryPointNode
  outputs [#2]
  properties {entryTag {tagName=Conversation.Entry.TheVisitor}}
 }
 Speech_C_4 {
  id=#2
  class=Speech_C
  outputs [
   #3
   #4
   #5
   #6
   #7
   #8
  ]
  properties {
   nodeName=""
   generalConfig {
    defaultMessage=You may call me The Visitor.
    contextualMessages [
     {
      message=You may call me The Visitor.
      contextRequirements [{requirement {class=FortControllerRequirement_IsFirstEverConversationWithNPC}}]
      priority {curve {rowName=Tandem.Speech.Priority.FirstEverInteraction}}
     }
     {
      message="You may not be my Order, but it is good to see you again."
      contextRequirements [{requirement {
         class=FortControllerRequirement_HasCID
         properties {allowedCIds [/BRCosmetics/Athena/Items/Cosmetics/Characters/CID_A_395_Athena_Commando_F_BinaryTwin]}
        }}]
      priority {curve {rowName=Tandem.Speech.Priority.OutfitReaction}}
     }
     {
      message="I, too, understand the intoxicating nature of vengeance."
      contextRequirements [{requirement {
         class=FortControllerRequirement_HasCID
         properties {allowedCIds [&Character_UndergroundRebel_Fashion]}
        }}]
      priority {curve {rowName=Tandem.Speech.Priority.OutfitReaction}}
     }
     {
      message="My apologies, but you seem a little out of your depth."
      contextRequirements [{requirement {class=FortControllerRequirement_HasCID, properties {allowedCIds [&Character_BrightDisk]}}}]
      priority {curve {rowName=Tandem.Speech.Priority.OutfitReaction}}
     }
     {
      message=None of them truly understands what you are...
      contextRequirements [{requirement {class=FortControllerRequirement_HasCID, properties {allowedCIds [/BRCosmetics/Athena/Items/Cosmetics/Characters/Character_SpacePlunge]}}}]
      priority {curve {rowName=Tandem.Speech.Priority.OutfitReaction}}
     }
     {
      message=I am here to help. However I can.
      contextRequirements [{requirement {class=FortControllerRequirement_CheckNPCMatchConversationCount, properties {nPCUniqueIdTag {tagName=AISpawnerData.Type.Tandem.TheVisitor}}}}]
      priority {curve {rowName=None}}
     }
     {
      message=The Dark Voyager must be stopped.
      contextRequirements [{requirement {class=FortControllerRequirement_CheckNPCMatchConversationCount, properties {nPCUniqueIdTag {tagName=AISpawnerData.Type.Tandem.TheVisitor}}}}]
      priority {curve {rowName=None}}
     }
     {
      message=I've come a long way. It is good to rest.
      contextRequirements [{requirement {class=FortControllerRequirement_CheckNPCMatchConversationCount, properties {nPCUniqueIdTag {tagName=AISpawnerData.Type.Tandem.TheVisitor}}}}]
      priority {curve {rowName=None}}
     }
     {
      message="I am The Visitor, one of The Seven."
      contextRequirements [{requirement {class=FortControllerRequirement_CheckNPCMatchConversationCount, properties {nPCUniqueIdTag {tagName=AISpawnerData.Type.Tandem.TheVisitor}}}}]
      priority {curve {rowName=None}}
     }
    ]
   }
  }
 }
 ConversationLinkNode_5 {
  id=#8
  class=ConversationLinkNode
  properties {remoteEntryTag {tagName=Conversation.Entry.Season39.Weekly.W13.Q01.SynthContact}}
 }
 ConversationLinkNode_3 {
  id=#7
  class=ConversationLinkNode
  properties {remoteEntryTag {tagName=Conversation.Entry.TheVisitor.Dialogue}}
 }
 ConversationLinkNode_2 {
  id=#6
  class=ConversationLinkNode
  properties {remoteEntryTag {tagName=Conversation.Entry.Season39.SynthContact.Story.03.01}}
 }
 ConversationLinkNode_7 {id=#5, class=ConversationLinkNode, properties {remoteEntryTag {tagName=Conversation.Entry.RandomServices}}}
 ConversationLinkNode_0 {id=#4, class=ConversationLinkNode, properties {remoteEntryTag {tagName=Conversation.Entry.BasicServices}}}
 ConversationLinkNode_1 {id=#3, class=ConversationLinkNode, properties {remoteEntryTag {tagName=Conversation.Entry.NPCLibrary.Quests}}}
]
